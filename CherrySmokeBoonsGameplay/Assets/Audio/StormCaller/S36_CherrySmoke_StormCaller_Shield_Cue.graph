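# Sound cue for the shield sound of the StormCaller boon in CherrySmoke.
# A Random node picks one of six equally weighted Shield wave variations, then a Modulator adds slight pitch and volume variation.
# The cue plays at reduced volume in the Fort_Loot_Collect sound class, with a concurrency override that limits it to one instance at a time.
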SoundCue /CherrySmokeBoonsGameplay/Assets/Audio/StormCaller/S36_CherrySmoke_StormCaller_Shield_Cue
root=Modulator_0
nodes [
 {id=Modulator_0, children [Random_0], properties {pitchMin=0.98, volumeMax=1.02, volumeMin=0.9}}
 {id=Random_0, children [WavePlayer_0,WavePlayer_2,WavePlayer_4,WavePlayer_5,WavePlayer_6,WavePlayer_7], properties {weights [1,1,1,1,1,1]}}
 {id=WavePlayer_0, properties {soundWaveAssetPtr=&/Boon_CherrySmoke_StormCaller_Shield_06}}
 {id=WavePlayer_2, properties {soundWaveAssetPtr=&/Boon_CherrySmoke_StormCaller_Shield_05}}
 {id=WavePlayer_4, properties {soundWaveAssetPtr=&/Boon_CherrySmoke_StormCaller_Shield_04}}
 {id=WavePlayer_5, properties {soundWaveAssetPtr=&/Boon_CherrySmoke_StormCaller_Shield_03}}
 {id=WavePlayer_6, properties {soundWaveAssetPtr=&/Boon_CherrySmoke_StormCaller_Shield_02}}
 {id=WavePlayer_7, properties {soundWaveAssetPtr=&/Boon_CherrySmoke_StormCaller_Shield_01}}
]
properties {
 bOverrideConcurrency=true
 maxDistance=2097152
 soundClassObject=&Fort_Loot_Collect
 volumeMultiplier=0.5141
 concurrencyOverrides {maxCount=1}
}
